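# Translucent UI material for an NPC portrait in the quest HUD, with discovered and hover states.
# The background blends a vertical gradient, radial glow and panning screen-aligned triangle grid.
# The portrait composites over it with a lower falloff tint, zooms on hover, and becomes a flat-colored silhouette when IsDiscovered is 0.
# Four rotated UV edge masks form a skewed border that shifts with hover; a static switch outputs only the background.
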Material /FortniteCoreUI/Athena/HUD/Quests/Art/M_UI_Quest_NPC-Portrait
graph {
 properties {
  bUsedWithStaticMesh=false
  blendMode=BLEND_Translucent
  materialDomain=MD_UI
  pixelDepthOffsetMode=PDOM_Legacy
  usedShadingModels=Default Lit
 }
 nodes [
  ScalarParameter() {defaultValue=1, parameterName=Opacity, sliderMax=1}
  n1 = MaterialFunctionCall(
   UVs=TextureCoordinate()
   CenterPosition=AppendVector(
    A=ScalarParameter() {defaultValue=0.5, group=background, parameterName=RadialGradient_u, sliderMin=0.5}
    B=ScalarParameter() {defaultValue=0.5, group=background, parameterName=RadialGradient_V, sliderMin=0.5}
   )
   Radius=ScalarParameter() {defaultValue=0.75, group=background, parameterName=Radial Gradient Radius_V}
   Density=ScalarParameter() {defaultValue=1.5, group=background, parameterName=Radial Gradient Density}
  ) {materialFunction=&RadialGradientExponential}
  n2 = LinearInterpolate(
   A=VectorParameter() {parameterName=Background_base_F, defaultValue=<rgba 0.0033 0.0185 0.3763 1>}.RGB
   B=VectorParameter() {group=background, parameterName=Background_base_T, defaultValue=<rgba 0 0.2159 1 1>}.RGB
   Alpha=ScalarParameter() {defaultValue=1, group=control, parameterName=IsDiscovered, sliderMax=1}
  )
  n3 = LinearInterpolate(
   A=VectorParameter() {group=background, parameterName=Background_topgradient_T}.RGB
   B=VectorParameter() {group=background, parameterName=Background_glow_T, defaultValue=<rgba 0 0.6654 1 1>}.RGB
   Alpha=ScalarParameter() {defaultValue=1, group=control, parameterName=IsDiscovered, sliderMax=1}
  )
  n4 = MaterialFunctionCall(Float2=MaterialFunctionCall() {materialFunction=&ScreenResolution}.Visible Resolution) {materialFunction=&BreakOutFloat2Components}
  n5 = Max(A=@n4.G)
  n6 = Multiply(
   A=AppendVector(A=Divide(A=@n4.R, B=@n5), B=Constant() {r=1})
   B=MaterialFunctionCall() {materialFunction=&ScreenAlignedUVs}."X 100%, Y 100%"
  )
  n7 = ScalarParameter() {defaultValue=-0.0571, parameterName=Tri Angle}
  n8 = ScalarParameter() {defaultValue=0.0025, parameterName=PanSpeed}
  n9 = Constant()
  n10 = Multiply(
   A=Panner(Coordinate=Rotator(Coordinate=@n6, Time=@n7) {speed=6.2832}, Speed=AppendVector(A=@n8, B=@n9))
   B=ScalarParameter() {defaultValue=6, parameterName=Tri Tile Amount}
  )
  n11 = Saturate(None=Multiply(
    A=TextureSample(UVs=@n10) {texture=/Game/Global/Textures/Default/Masks_and_Gradients/T-TriangleGrid}.R
    B=ScalarParameter() {defaultValue=1, parameterName=TriangleDensity, sliderMax=1}
   ))
  n12 = TextureCoordinate()
  n13 = LinearInterpolate(
   A=VectorParameter() {group=background, parameterName=Background_topgradient_T, defaultValue=<rgba 0 0.0578 0.6038 1>}.RGB
   B=LinearInterpolate(A=LinearInterpolate(A=@n2, B=@n3, Alpha=@n11), B=@n3, Alpha=Saturate(None=@n1.RadialGradientExponential))
   Alpha=Saturate(None=ComponentMask(None=@n12) {g=true})
  )
  n14 = MaterialFunctionCall(Float2=MaterialFunctionCall() {materialFunction=&GetUserInterfaceUV}.Pixel Size) {materialFunction=&BreakOutFloat2Components}
  n15 = Max(A=@n14.G)
  n16 = MaterialFunctionCall(Float2=TextureCoordinate() {coordinateIndex=4}) {materialFunction=&BreakOutFloat2Components}
  n17 = AppendVector(A=Multiply(A=Divide(A=@n14.R, B=@n15), B=@n16.R), B=@n16.G)
  n18 = AppendVector(A=ScalarParameter() {group=Item, parameterName=Itemposition_U}, B=ScalarParameter() {group=Item, parameterName=Itemposition_V})
  n19 = ScalarParameter() {defaultValue=1.01, parameterName=hoverZoom}
  n20 = ScalarParameter() {group=control, parameterName=hover, sliderMax=1, sliderMin=0.1}
  n21 = TextureSampleParameter2D(UVs=MaterialFunctionCall(UVs=Add(A=@n17, B=@n18), Texture Scale=LinearInterpolate(B=@n19, Alpha=@n20) {constA=1, constB=1.01}) {materialFunction=&ScaleUVsByCenter}.UVs) {parameterName=NPC-Portrait, texture=/Game/Athena/HUD/Quests/Art/T_NPC_Default}
  n22 = TextureCoordinate()
  n23 = Multiply(
   A=Subtract(A=ComponentMask(None=@n22) {g=true}, B=ScalarParameter() {defaultValue=0.4, group=Item, parameterName=FallOff_Placement, sliderMax=1}) {constB=0.5}
   B=LinearInterpolate(
    A=ScalarParameter() {defaultValue=1, group=Item, parameterName=FallOff_Strenght}
    B=ScalarParameter() {defaultValue=1, group=Item, parameterName=FallOff_Strenght_Hovered}
    Alpha=ScalarParameter() {defaultValue=1, group=control, parameterName=IsDiscovered, sliderMax=1}
   ) {constB=0}
  ) {constB=2}
  n24 = LinearInterpolate(
   A=LinearInterpolate(
    A=VectorParameter() {parameterName=Background_topgradient_F, defaultValue=<rgba 0 0.004 0.0723 1>}.RGB
    B=@n21.RGB
    Alpha=ScalarParameter() {defaultValue=1, group=control, parameterName=IsDiscovered, sliderMax=1}
   )
   B=VectorParameter() {group=background, parameterName=Background_topgradient_T, defaultValue=<rgba 0 0.0578 0.6038 1>}.RGB
   Alpha=Multiply(A=@n21.A, B=Saturate(None=@n23))
  )
  n25 = ScalarParameter() {defaultValue=0.002, group=Border Angles, parameterName=Top-Angle}
  n26 = ScalarParameter() {defaultValue=-0.002, group=Border Angles, parameterName=Top-Angle-hover}
  n27 = ScalarParameter() {group=control, parameterName=hover, sliderMax=1}
  n28 = ComponentMask(None=Rotator(Coordinate=TextureCoordinate() {coordinateIndex=4}, Time=LinearInterpolate(A=@n25, B=@n26, Alpha=@n27)) {speed=6.2832}) {g=true}
  n29 = LinearInterpolate(
   A=ScalarParameter() {defaultValue=0.035, group=Border_Range, parameterName=Top-RangeOffset, sliderMax=1}
   B=ScalarParameter() {defaultValue=0.035, group=Border_Range, parameterName=Top-RangeOffset-hover, sliderMax=1}
   Alpha=ScalarParameter() {group=control, parameterName=hover, sliderMax=1}
  )
  n30 = Saturate(None=Multiply(A=Subtract(A=@n28, B=@n29), B=ScalarParameter() {defaultValue=512, group=Border sharpness, parameterName=HorizontalSharpness}))
  n31 = ScalarParameter() {defaultValue=-0.002, group=Border Angles, parameterName=Left-Angle}
  n32 = ScalarParameter() {defaultValue=0.001, group=Border Angles, parameterName=Left-Angle-hover}
  n33 = ScalarParameter() {group=control, parameterName=hover, sliderMax=1}
  n34 = ComponentMask(None=Rotator(Coordinate=TextureCoordinate() {coordinateIndex=4}, Time=LinearInterpolate(A=@n31, B=@n32, Alpha=@n33)) {speed=6.2832}) {r=true}
  n35 = LinearInterpolate(
   A=ScalarParameter() {defaultValue=0.035, group=Border_Range, parameterName=Left-RangeOffset, sliderMax=1}
   B=ScalarParameter() {defaultValue=0.035, group=Border_Range, parameterName=Left-RangeOffset-hover, sliderMax=1}
   Alpha=ScalarParameter() {group=control, parameterName=hover, sliderMax=1}
  )
  n36 = Saturate(None=Multiply(A=Subtract(A=@n34, B=@n35), B=ScalarParameter() {defaultValue=512, group=Border sharpness, parameterName=VerticalSharpness}))
  n37 = ScalarParameter() {defaultValue=0.5025, group=Border Angles, parameterName=Bot-Angle}
  n38 = ScalarParameter() {defaultValue=0.4975, group=Border Angles, parameterName=Bot-Angle-hover}
  n39 = ScalarParameter() {group=control, parameterName=hover, sliderMax=1}
  n40 = ComponentMask(None=Rotator(Coordinate=TextureCoordinate() {coordinateIndex=4}, Time=LinearInterpolate(A=@n37, B=@n38, Alpha=@n39)) {speed=6.2832}) {g=true}
  n41 = LinearInterpolate(
   A=ScalarParameter() {defaultValue=0.035, group=Border_Range, parameterName=Bot-RangeOffset, sliderMax=1}
   B=ScalarParameter() {defaultValue=0.035, group=Border_Range, parameterName=Bot-RangeOffset-hover, sliderMax=1}
   Alpha=ScalarParameter() {group=control, parameterName=hover, sliderMax=1}
  )
  n42 = Saturate(None=Multiply(A=Subtract(A=@n40, B=@n41), B=ScalarParameter() {group=Border sharpness, parameterName=HorizontalSharpness}))
  n43 = ScalarParameter() {defaultValue=0.4975, group=Border Angles, parameterName=Right-Angle}
  n44 = ScalarParameter() {defaultValue=0.5015, group=Border Angles, parameterName=Right-Angle-hover}
  n45 = ScalarParameter() {group=control, parameterName=hover, sliderMax=1}
  n46 = ComponentMask(None=Rotator(Coordinate=TextureCoordinate() {coordinateIndex=4}, Time=LinearInterpolate(A=@n43, B=@n44, Alpha=@n45)) {speed=6.2832}) {r=true}
  n47 = LinearInterpolate(
   A=ScalarParameter() {defaultValue=0.035, group=Border_Range, parameterName=Right-RangeOffset, sliderMax=1}
   B=ScalarParameter() {defaultValue=0.035, group=Border_Range, parameterName=Right-RangeOffset-hover, sliderMax=1}
   Alpha=ScalarParameter() {group=control, parameterName=hover, sliderMax=1}
  )
  n48 = Saturate(None=Multiply(A=Subtract(A=@n46, B=@n47), B=ScalarParameter() {group=Border sharpness, parameterName=VerticalSharpness}))
  n49 = Saturate(None=DotProduct(A=DotProduct(A=@n30, B=@n36), B=DotProduct(A=@n42, B=@n48)))
  n50 = Multiply(
   A=Saturate(None=OneMinus(None=@n49))
   B=LinearInterpolate(
    A=ScalarParameter() {group=Border Alpha, parameterName=BorderAlpha, sliderMax=1}
    B=ScalarParameter() {defaultValue=1, group=Border Alpha, parameterName=HoveredBorderAlpha, sliderMax=1, sliderMin=0.01}
    Alpha=ScalarParameter() {group=control, parameterName=hover, sliderMax=1}
   ) {constB=0}
  )
  StaticSwitchParameter(True=@n13, False=LinearInterpolate(A=LinearInterpolate(A=@n13, B=@n24, Alpha=@n21.A), Alpha=Saturate(None=@n50))) {parameterName=bIsBackgroundOnly}
 ]
}
