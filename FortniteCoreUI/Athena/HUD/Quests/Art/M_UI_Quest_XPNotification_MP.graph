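# Translucent UI material, by its name the background for the quest XP notification in the Athena HUD.
# Color parameters and Lerp nodes build a gradient from horizontal UI UV and texture coordinates, tuned by sharpness and start/end offset scalars.
# MF_RoundedSquare takes 9-slice UVs and a curvature value, and its linear gradient is scaled by edge sharpness and saturated into a rounded-corner mask.
# The file shows no links from these chains to the material outputs.
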Material /FortniteCoreUI/Athena/HUD/Quests/Art/M_UI_Quest_XPNotification_MP
graph {
 properties {
  bUsedWithStaticMesh=false
  blendMode=BLEND_Translucent
  materialDomain=MD_UI
  pixelDepthOffsetMode=PDOM_Legacy
  usedShadingModels=Default Lit
 }
 nodes [
  n1 = ComponentMask(None=MaterialFunctionCall() {materialFunction=&GetUserInterfaceUV}.Normalized UV) {r=true}
  n2 = ScalarParameter() {defaultValue=0.7, parameterName=Gradient_Sharpness, sortPriority=40}
  n3 = TextureCoordinate()
  n4 = ScalarParameter() {defaultValue=2, parameterName=Gradient_End_Offset, sortPriority=20}
  LinearInterpolate(
   A=VectorParameter() {parameterName=RGBA2, defaultValue=<rgba 0.0885 0 0.0838 1>}.RGB
   B=VectorParameter() {parameterName=RGBA3, defaultValue=<rgba 0.027 0 0.0885 1>}.RGB
   Alpha=Multiply(A=Multiply(A=@n1, B=@n2), B=Multiply(A=@n3, B=@n4))
  )
  LinearInterpolate()
  Multiply(A=Multiply(A=TextureCoordinate(), B=ScalarParameter() {defaultValue=0.5, parameterName=Gradient_Start_Offset, sortPriority=30}))
  n5 = MaterialFunctionCall!(
   UV=MaterialFunctionCall() {materialFunction=&GetUserInterfaceUV}."9-Slice UV (No Tiling)"
   Size
   Curvature=ScalarParameter() {defaultValue=0.25, group=Rounded corners, parameterName=Curvature}
  ) {materialFunction=/Game/Effects/Fort_Effects/MaterialFunctions/MF_RoundedSquare}
  Saturate(None=Multiply(A=ComponentMask(None=@n5.Linear Gradient) {g=true}, B=ScalarParameter() {defaultValue=4, group=Rounded corners, parameterName=EdgeSharpness}))
  VectorParameter() {parameterName=RGBA1, defaultValue=<rgba 0.0031 0 0.0052 1>}
 ]
}
