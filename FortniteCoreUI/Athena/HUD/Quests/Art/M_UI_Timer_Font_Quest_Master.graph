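# Translucent UI master material for the quest timer font in the Athena HUD, providing a text color that can shift on hover.
# A Hovered scalar parameter linearly blends between the RGB-Default and RGB-Hovered vector parameters, both white by default.
# The graph uses no textures or animation, so material instances or runtime parameter changes supply any visible color difference.
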Material /FortniteCoreUI/Athena/HUD/Quests/Art/M_UI_Timer_Font_Quest_Master
graph {
 properties {
  bUsedWithStaticMesh=false
  blendMode=BLEND_Translucent
  materialDomain=MD_UI
  pixelDepthOffsetMode=PDOM_Legacy
  usedShadingModels=Default Lit
 }
 nodes [LinearInterpolate(
   A=VectorParameter() {parameterName=RGB-Default, defaultValue=<rgba 1 1 1 1>}.RGB
   B=VectorParameter() {parameterName=RGB-Hovered, defaultValue=<rgba 1 1 1 1>}.RGB
   Alpha=ScalarParameter() {parameterName=Hovered}
  )]
}
