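# Translucent UI material for the quest XP notification: a two-color gradient panel with animated triangles and rounded corners.
# A triangle grid texture uses aspect-corrected screen UVs, rotated and slowly panned.
# A horizontal widget-UV mask fades the triangles, which are color-dodged over the horizontal gradient.
# MF_RoundedSquare on 9-slice UVs builds a rounded-corner mask with adjustable curvature and edge sharpness.
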Material /FortniteCoreUI/Athena/HUD/Quests/Art/M_Ui_Quest_XPNotification
graph {
 properties {
  bUsedWithStaticMesh=false
  blendMode=BLEND_Translucent
  materialDomain=MD_UI
  pixelDepthOffsetMode=PDOM_Legacy
  usedShadingModels=Default Lit
 }
 nodes [
  n1 = MaterialFunctionCall(Float2=MaterialFunctionCall() {materialFunction=&ScreenResolution}.Visible Resolution) {materialFunction=&BreakOutFloat2Components}
  n2 = Max(A=@n1.G)
  n3 = Multiply(
   A=AppendVector(A=Divide(A=@n1.R, B=@n2), B=Constant() {r=1})
   B=MaterialFunctionCall() {materialFunction=&ScreenAlignedUVs}."X 100%, Y 100%"
  )
  n4 = ScalarParameter() {defaultValue=-0.0278, group=Triangles, parameterName=Tri Angle}
  n5 = ScalarParameter() {defaultValue=0.0025, group=Triangles, parameterName=PanSpeed}
  n6 = Constant()
  n7 = Multiply(
   A=Panner(Coordinate=Rotator(Coordinate=@n3, Time=@n4) {speed=6.2832}, Speed=AppendVector(A=@n5, B=@n6))
   B=ScalarParameter() {defaultValue=0.5, group=Triangles, parameterName=Tri Tile Amount}
  )
  n8 = Saturate(None=Multiply(
    A=TextureSample(UVs=@n7) {texture=/Game/Global/Textures/Default/Masks_and_Gradients/T-TriangleGrid}.R
    B=ScalarParameter() {defaultValue=0.2, group=Triangles, parameterName=TriangleDensity}
   ))
  n9 = Subtract(
   A=MaterialFunctionCall() {materialFunction=&GetUserInterfaceUV}.Normalized UV
   B=ScalarParameter() {group=Triangles, parameterName=TriangleMask_Position}
  )
  n10 = Saturate(None=Multiply(A=ComponentMask(None=@n9) {r=true}, B=ScalarParameter() {defaultValue=1.5, group=Triangles, parameterName=TriangleMask_sharpness}))
  n11 = VectorParameter() {parameterName=RGBA1, defaultValue=<rgba 0.016 0.1746 0.7913 1>}
  n12 = VectorParameter() {parameterName=RGBA2, defaultValue=<rgba 0 0.004 0.0723 1>}
  n13 = Subtract(
   A=MaterialFunctionCall() {materialFunction=&GetUserInterfaceUV}.Normalized UV
   B=ScalarParameter() {parameterName=Gradient_Position}
  )
  n14 = Saturate(None=Multiply(A=ComponentMask(None=@n13) {r=true}, B=ScalarParameter() {defaultValue=0.75, parameterName=Gradient_Sharpness}))
  Multiply(A=MaterialFunctionCall!(Blend=Multiply(A=@n8, B=@n10), Base=LinearInterpolate(A=@n11.RGB, B=@n12.RGB, Alpha=@n14)) {materialFunction=/Engine/Functions/Engine_MaterialFunctions03/Blends/Blend_ColorDodge}.Result)
  n15 = MaterialFunctionCall!(
   UV=MaterialFunctionCall() {materialFunction=&GetUserInterfaceUV}."9-Slice UV (No Tiling)"
   Size
   Curvature=ScalarParameter() {defaultValue=0.25, group=Rounded corners, parameterName=Curvature}
  ) {materialFunction=/Game/Effects/Fort_Effects/MaterialFunctions/MF_RoundedSquare}
  Saturate(None=Multiply(
    A=ComponentMask(None=@n15.Linear Gradient) {g=true}
    B=ScalarParameter() {defaultValue=4, group=Rounded corners, parameterName=EdgeSharpness}
   ))
 ]
}
